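# Physics asset for the FaceAcc part of the BoltBolo male medium body cosmetic, giving a dangling tie physical bodies and joints.
# Six small tie capsules form a chain from dyn_tie_mstr, which joins spine_05, with narrow swing limits and almost no twist between links.
# Larger capsules on the spine, neck, clavicles and upper arms approximate the torso, likely as tie colliders, and their joints leave rotation free.
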PhysicsAsset /BRCosmetics/Characters/Player/Male/Medium/Bodies/M_MED_FoundKelp_BoltBolo/Meshes/Parts/M_MED_FoundKelp_BoltBolo_FaceAcc_Physics
bodies {
 clavicle_l [{shapeType=Capsule, length=18, radius=10, shapeName=clavicle_l_capsule, center=<6 -1 2>, rotation=<rot 80 -0.0003 -0.0003>}]
 clavicle_r [{shapeType=Capsule, length=18, radius=10, shapeName=clavicle_r_capsule, center=<-6 1 -2>, rotation=<rot 80 -0.0003 -0.0004>}]
 dyn_tie_01 [{shapeType=Capsule, length=3, radius=3, center=<4 0 2>, rotation=<rot 0 0 90>}]
 dyn_tie_02 [{shapeType=Capsule, length=3, radius=3, shapeName=dyn_tie_02_capsule, center=<4 0 2>, rotation=<rot 0 0 90>}]
 dyn_tie_03 [{shapeType=Capsule, length=3, radius=3, shapeName=dyn_tie_03_capsule, center=<4 0 2>, rotation=<rot 0 0 90>}]
 dyn_tie_04 [{shapeType=Capsule, length=3, radius=3, shapeName=dyn_tie_04_capsule, center=<4 0 2>, rotation=<rot 0 0 90>}]
 dyn_tie_05 [{shapeType=Capsule, length=3, radius=3, shapeName=dyn_tie_05_capsule, center=<4 0 2>, rotation=<rot 0 0 90>}]
 dyn_tie_06 [{shapeType=Capsule, length=3, radius=3, shapeName=dyn_tie_06_capsule, center=<4 0 2>, rotation=<rot 0 0 90>}]
 dyn_tie_mstr [{shapeType=Capsule, length=0.505, radius=0.505, shapeName=dyn_tie_mstr_capsule, rotation=<rot 0 -90 0>}]
 neck_01 [{shapeType=Capsule, length=8, radius=4.5, shapeName=neck_01_capsule, center=<3.6683 -0.2876 0>, rotation=<rot 90 0 0>}]
 spine_03 [{shapeType=Capsule, length=6, radius=12, shapeName=spine_04_capsule, center=<0.229 2 0>}]
 spine_04 [{shapeType=Capsule, length=6, radius=16, shapeName=spine_04_capsule, center=<3.5 0.5355 0>, rotation=<rot 0 -90 0>}]
 spine_05 [{shapeType=Capsule, length=10, radius=14, shapeName=spine_05_capsule, center=<0 0.209 0>}]
 upperarm_l [{shapeType=Capsule, length=22, radius=7, shapeName=upperarm_l_capsule, center=<12 -0.5 0>, rotation=<rot 90 0 0>}]
 upperarm_r [{shapeType=Capsule, length=22, radius=7, shapeName=upperarm_r_capsule, center=<-12 0.5 0>, rotation=<rot 90 0 0>}]
}
constraints [
 {bone1Name=dyn_tie_02, bone2Name=dyn_tie_01, swing1LimitDegrees=25, swing2LimitDegrees=25, twistLimitDegrees=1}
 {bone1Name=dyn_tie_03, bone2Name=dyn_tie_02, swing1LimitDegrees=25, swing2LimitDegrees=25, twistLimitDegrees=1}
 {bone1Name=dyn_tie_04, bone2Name=dyn_tie_03, swing1LimitDegrees=25, swing2LimitDegrees=25, twistLimitDegrees=1}
 {bone1Name=dyn_tie_05, bone2Name=dyn_tie_04, swing1LimitDegrees=25, swing2LimitDegrees=25, twistLimitDegrees=1}
 {bone1Name=dyn_tie_06, bone2Name=dyn_tie_05, swing1LimitDegrees=25, swing2LimitDegrees=25, twistLimitDegrees=1}
 {bone1Name=dyn_tie_01, bone2Name=dyn_tie_mstr, swing1LimitDegrees=25, swing2LimitDegrees=25, twistLimitDegrees=1}
 {
  bone1Name=dyn_tie_mstr
  bone2Name=spine_05
  swing1LimitDegrees=45
  swing1Motion=Free
  swing2LimitDegrees=45
  swing2Motion=Free
  twistLimitDegrees=45
  twistMotion=Free
 }
 {
  bone1Name=spine_05
  bone2Name=spine_04
  swing1LimitDegrees=45
  swing1Motion=Free
  swing2LimitDegrees=45
  swing2Motion=Free
  twistLimitDegrees=45
  twistMotion=Free
 }
 {
  bone1Name=neck_01
  bone2Name=spine_04
  swing1LimitDegrees=45
  swing1Motion=Free
  swing2LimitDegrees=45
  swing2Motion=Free
  twistLimitDegrees=45
  twistMotion=Free
 }
 {
  bone1Name=clavicle_l
  bone2Name=spine_04
  swing1LimitDegrees=45
  swing1Motion=Free
  swing2LimitDegrees=45
  swing2Motion=Free
  twistLimitDegrees=45
  twistMotion=Free
 }
 {
  bone1Name=upperarm_l
  bone2Name=spine_04
  swing1LimitDegrees=45
  swing1Motion=Free
  swing2LimitDegrees=45
  swing2Motion=Free
  twistLimitDegrees=45
  twistMotion=Free
 }
 {
  bone1Name=clavicle_r
  bone2Name=spine_04
  swing1LimitDegrees=45
  swing1Motion=Free
  swing2LimitDegrees=45
  swing2Motion=Free
  twistLimitDegrees=45
  twistMotion=Free
 }
 {
  bone1Name=upperarm_r
  bone2Name=spine_04
  swing1LimitDegrees=45
  swing1Motion=Free
  swing2LimitDegrees=45
  swing2Motion=Free
  twistLimitDegrees=45
  twistMotion=Free
 }
 {
  bone1Name=spine_04
  bone2Name=spine_03
  swing1LimitDegrees=45
  swing1Motion=Free
  swing2LimitDegrees=45
  swing2Motion=Free
  twistLimitDegrees=45
  twistMotion=Free
 }
]
