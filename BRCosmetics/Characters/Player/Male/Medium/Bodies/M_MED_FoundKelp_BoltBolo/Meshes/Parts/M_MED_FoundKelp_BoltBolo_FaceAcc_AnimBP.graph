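# Animation Blueprint for the face accessory part of the FoundKelp BoltBolo character, adding physics-driven secondary motion.
# It copies the attached parent's pose and curves, then layers in the upper-arm branches from the owner's skeletal mesh, refreshed on initialize and each update.
# A RigidBody node simulates the result with a dedicated physics asset in spine_05 space, using overridden gravity, capped velocities and accelerations, and drag.
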AnimBlueprint /BRCosmetics/Characters/Player/Male/Medium/Bodies/M_MED_FoundKelp_BoltBolo/Meshes/Parts/M_MED_FoundKelp_BoltBolo_FaceAcc_AnimBP
target_skeleton=&/M_MED_FoundKelp_BoltBolo_FaceAcc_Skeleton
graphs [
 AnimGraph {nodes [
   n1 = Copy Pose From Mesh() {bCopyCurves=true, bUseAttachedParent=true}
   n2 = Copy Pose From Mesh(SourceMeshComponent=Get body Mesh() -> (body Mesh:Skeletal Mesh Component Object Reference))
   n3 = RigidBody!(
    ComponentPose=Local To Component(LocalPose=Layered blend per bone(BasePose=@n1, BlendPoses_0=@n2) {layerSetup [{branchFilters [{boneName=upperarm_l}, {boneName=upperarm_r}]}]})
    Alpha
    OverrideWorldGravity:Vector="0,0,-980"
    ExternalForce
   ) -> (Pose) {
    bClampLinearTranslationLimitToRefPose=true
    bOverrideWorldGravity=true
    overridePhysicsAsset=&/M_MED_FoundKelp_BoltBolo_FaceAcc_Physics
    simulationSpace=BaseBoneSpace
    alphaBoolBlend {blendOption=Linear}
    baseBoneRef {boneName=spine_05}
    overrideWorldGravity {z=-980}
    simSpaceSettings {
     maxAngularAcceleration=40
     maxAngularVelocity=8
     maxLinearAcceleration=3500
     maxLinearVelocity=700
     worldAlpha=0.5
     externalLinearDragV=<3 3 3>
    }
   }
   Output Pose(Result=Component To Local(ComponentPose=@n3)) {layerGroup=None}
  ]}
 EventGraph {nodes [
   Event BlueprintUpdateAnimation()
   > Set body Mesh(body Mesh:Skeletal Mesh Component Object Reference=GetOwnerSkeletalMeshComponent()) -> (Output_Get:Skeletal Mesh Component Object Reference)
   Event BlueprintInitializeAnimation()
   > Set body Mesh(body Mesh=GetOwnerSkeletalMeshComponent())
  ]}
]
variables [body Mesh {type=SkeletalMeshComponent}]
